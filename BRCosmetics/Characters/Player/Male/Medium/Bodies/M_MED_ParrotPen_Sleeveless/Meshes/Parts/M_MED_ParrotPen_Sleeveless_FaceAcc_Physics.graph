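# Physics asset for the face accessory part of the male medium ParrotPen Sleeveless cosmetic body, set up for hair bone physics.
# Capsule bodies cover spine_05, head, two hair base bones and five dyn_hair bones, with free-rotation constraints linking head to spine_05 and both bases to head.
# Four strands attach in parallel to base_hair01 and dyn_hairDw to base_hairDw, each with narrow swing limits and zero-degree twist limits.
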PhysicsAsset /BRCosmetics/Characters/Player/Male/Medium/Bodies/M_MED_ParrotPen_Sleeveless/Meshes/Parts/M_MED_ParrotPen_Sleeveless_FaceAcc_Physics
bodies {
 base_hair01 [{shapeType=Capsule, length=0.505, radius=0.505, shapeName=base_hair01_capsule, rotation=<rot 0 -90 0>}]
 base_hairDw [{shapeType=Capsule, length=0.505, radius=0.505, shapeName=base_hairDw_capsule, rotation=<rot 0 -90 0>}]
 dyn_hair01 [{shapeType=Capsule, length=1, radius=1, shapeName=dyn_hair01_capsule, center=<5 0 0>, rotation=<rot 90 0 0>}]
 dyn_hair02 [{shapeType=Capsule, length=1, radius=1, shapeName=dyn_hair02_capsule, center=<5 0 0>, rotation=<rot 90 0 0>}]
 dyn_hair03 [{shapeType=Capsule, length=1, radius=1, shapeName=dyn_hair03_capsule, center=<5 0 0>, rotation=<rot 90 0 0>}]
 dyn_hair04 [{shapeType=Capsule, length=1, radius=1, shapeName=dyn_hair04_capsule, center=<5 0 0>, rotation=<rot 90 0 0>}]
 dyn_hairDw [{shapeType=Capsule, length=2.125, radius=3.875, shapeName=dyn_hairDw_capsule, center=<6.0852 0 0>, rotation=<rot 0 0 -90>}]
 head [{shapeType=Capsule, length=2, radius=9.3055, center=<6.0784 2.335 0>, rotation=<rot 90 0 -11.2497>}]
 spine_05 [{shapeType=Capsule, length=13.3175, radius=9.8175, shapeName=spine_05_capsule, center=<7.6782 -3.8849 0>, rotation=<rot 0 -90 0>}]
}
constraints [
 {
  bone1Name=base_hair01
  bone2Name=head
  swing1LimitDegrees=45
  swing1Motion=Free
  swing2LimitDegrees=45
  swing2Motion=Free
  twistLimitDegrees=45
  twistMotion=Free
 }
 {
  bone1Name=base_hairDw
  bone2Name=head
  swing1LimitDegrees=45
  swing1Motion=Free
  swing2LimitDegrees=45
  swing2Motion=Free
  twistLimitDegrees=45
  twistMotion=Free
 }
 {bone1Name=dyn_hairDw, bone2Name=base_hairDw, swing1LimitDegrees=5, swing2LimitDegrees=5, twistLimitDegrees=0}
 {
  bone1Name=head
  bone2Name=spine_05
  swing1LimitDegrees=45
  swing1Motion=Free
  swing2LimitDegrees=45
  swing2Motion=Free
  twistLimitDegrees=45
  twistMotion=Free
 }
 {bone1Name=dyn_hair01, bone2Name=base_hair01, swing1LimitDegrees=12, swing2LimitDegrees=12, twistLimitDegrees=0}
 {bone1Name=dyn_hair02, bone2Name=base_hair01, swing1LimitDegrees=12, swing2LimitDegrees=12, twistLimitDegrees=0}
 {bone1Name=dyn_hair03, bone2Name=base_hair01, swing1LimitDegrees=12, swing2LimitDegrees=12, twistLimitDegrees=0}
 {bone1Name=dyn_hair04, bone2Name=base_hair01, swing1LimitDegrees=12, swing2LimitDegrees=12, twistLimitDegrees=0}
]
